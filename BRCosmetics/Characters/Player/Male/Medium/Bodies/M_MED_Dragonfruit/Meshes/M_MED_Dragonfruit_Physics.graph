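# Physics asset for the male medium Dragonfruit player body, covering its dynamic skirt and strap bones.
# Capsules on the pelvis and both thighs form the body shapes around ten two-bone skirt chains (A–J) and a five-bone strap chain, all rooted at the pelvis.
# Skirt and strap bones mostly use capsules, with boxes on several segments.
# Constraints keep most skirt swing very tight and the strap slightly looser, while thighs and SkirtA01 allow wider motion.
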PhysicsAsset /BRCosmetics/Characters/Player/Male/Medium/Bodies/M_MED_Dragonfruit/Meshes/M_MED_Dragonfruit_Physics
bodies {
 dyn_SkirtA01 [{shapeType=Capsule, length=2, radius=1.5, shapeName=dyn_SkirtA01_capsule, center=<0 0 -1.693>}]
 dyn_SkirtA02 [{shapeType=Capsule, length=6, radius=2.5, shapeName=dyn_SkirtA02_capsule, center=<0 0.5974 -3.8555>, rotation=<rot -1.1868 0 0>}]
 dyn_SkirtB01 [{shapeType=Capsule, length=1, radius=2.5, shapeName=dyn_SkirtB01_capsule, center=<0 0 -1.693>}]
 dyn_SkirtB02 [{shapeType=Box, extentX=11, extentY=2.5, extentZ=11.5, shapeName=dyn_SkirtJ02_box, center=<0 0.2474 -3.919>, rotation=<rot 0 0 -6.4623>}]
 dyn_SkirtC01 [{shapeType=Capsule, length=1, radius=2.5, shapeName=dyn_SkirtC01_capsule, center=<0 0 -1.693>}]
 dyn_SkirtC02 [{
   shapeType=Box
   extentX=9.5
   extentY=3
   extentZ=10.5
   shapeName=dyn_SkirtC02_box
   center=<-0.0001 -1.008 -4.3864>
   rotation=<rot 0 -0.0001 9.0425>
  }]
 dyn_SkirtD01 [{shapeType=Capsule, length=1, radius=2.5, shapeName=dyn_SkirtD01_capsule, center=<0 0 -1.693>}]
 dyn_SkirtD02 [{shapeType=Capsule, length=5, radius=3.5, shapeName=dyn_SkirtD02_capsule, center=<0 -0.0001 -3.8555>}]
 dyn_SkirtE01 [{shapeType=Capsule, length=1, radius=2.5, shapeName=dyn_SkirtE01_capsule, center=<0 0 -1.693>}]
 dyn_SkirtE02 [{shapeType=Capsule, length=5.5, radius=3, shapeName=dyn_SkirtE02_capsule, center=<0 -0.0001 -3.8555>}]
 dyn_SkirtF01 [{shapeType=Capsule, length=1, radius=2.5, shapeName=dyn_SkirtF01_capsule, center=<0 0 -1.693>}]
 dyn_SkirtF02 [{shapeType=Capsule, length=5.5, radius=3, shapeName=dyn_SkirtF02_capsule, center=<0 -0.0001 -3.8555>}]
 dyn_SkirtG01 [{shapeType=Capsule, length=1, radius=2.5, shapeName=dyn_SkirtG01_capsule, center=<0 0 -1.693>}]
 dyn_SkirtG02 [{shapeType=Capsule, length=5.25, radius=3.25, shapeName=dyn_SkirtG02_capsule, center=<0 0.7467 -3.8555>}]
 dyn_SkirtH01 [{shapeType=Capsule, length=1, radius=2.5, shapeName=dyn_SkirtH01_capsule, center=<0 0 -1.693>}]
 dyn_SkirtH02 [{shapeType=Capsule, length=6.25, radius=2.25, shapeName=dyn_SkirtH02_capsule, center=<0 -0.0001 -3.8555>}]
 dyn_SkirtI01 [{shapeType=Capsule, length=0, radius=3.5, shapeName=dyn_SkirtI01_capsule, center=<0 0 -1.693>}]
 dyn_SkirtI02 [{
   shapeType=Box
   extentX=12
   extentY=3
   extentZ=10.5
   shapeName=dyn_SkirtC02_box
   center=<-0.0001 -1.008 -4.3864>
   rotation=<rot 0 -0.0001 9.0425>
  }]
 dyn_SkirtJ01 [{shapeType=Capsule, length=4.5, radius=3.25, shapeName=dyn_SkirtJ01_capsule, center=<-0.0001 -2.1766 -1.693>, rotation=<rot 90 0.0395 0.0394>}]
 dyn_SkirtJ02 [{shapeType=Box, extentX=11, extentY=1.6696, extentZ=12.5, shapeName=dyn_SkirtJ02_box, center=<0 0.2474 -3.919>, rotation=<rot 0 0 -6.4623>}]
 dyn_Strap01 [{shapeType=Capsule, length=3.75, radius=2.5, shapeName=dyn_Strap01_capsule, center=<0.0004 0 -2.4845>}]
 dyn_Strap02 [{shapeType=Box, extentX=7.5, extentY=3, extentZ=9, shapeName=dyn_Strap02_box, center=<0 0.0524 -4.0374>}]
 dyn_Strap03 [{shapeType=Capsule, length=9.6536, radius=3.1195, shapeName=dyn_Strap03_capsule, center=<0 0 -5.2018>}]
 dyn_Strap04 [{shapeType=Capsule, length=9.6536, radius=3.1195, shapeName=dyn_Strap04_capsule, center=<0 0 -5.2018>}]
 dyn_Strap05 [{shapeType=Capsule, length=9.6536, radius=3.1195, shapeName=dyn_Strap05_capsule, center=<0 0 -5.2018>}]
 pelvis [{shapeType=Capsule, length=8.9797, radius=8.4943, center=<-5.094 0.0172 -0.576>, rotation=<rot 0.3252 -90.0695 -0.0608>}]
 thigh_l [{shapeType=Capsule, length=23.2793, radius=8.0968, center=<-20.4587 -0.5634 0.4534>, rotation=<rot 0 -88.822 -87.6619>}]
 thigh_r [{shapeType=Capsule, length=28.6489, radius=8.4772, center=<20.534 0.2791 -0.3>, rotation=<rot 0 -88.822 -87.6619>}]
}
constraints [
 {bone1Name=thigh_l, bone2Name=pelvis, swing1LimitDegrees=45, swing2LimitDegrees=45, twistLimitDegrees=45}
 {bone1Name=thigh_r, bone2Name=pelvis, swing1LimitDegrees=45, swing2LimitDegrees=45, twistLimitDegrees=45}
 {bone1Name=dyn_SkirtA01, bone2Name=pelvis, swing1LimitDegrees=45, swing2LimitDegrees=45, twistLimitDegrees=45}
 {bone1Name=dyn_SkirtA02, bone2Name=dyn_SkirtA01, swing1LimitDegrees=5, swing2LimitDegrees=5, twistLimitDegrees=15}
 {bone1Name=dyn_SkirtB01, bone2Name=pelvis, swing1LimitDegrees=5, swing2LimitDegrees=5, twistLimitDegrees=20}
 {bone1Name=dyn_SkirtB02, bone2Name=dyn_SkirtB01, swing1LimitDegrees=5, swing2LimitDegrees=5, twistLimitDegrees=20}
 {bone1Name=dyn_SkirtC01, bone2Name=pelvis, swing1LimitDegrees=5, swing2LimitDegrees=5, twistLimitDegrees=20}
 {bone1Name=dyn_SkirtC02, bone2Name=dyn_SkirtC01, swing1LimitDegrees=5, swing2LimitDegrees=5, twistLimitDegrees=20}
 {bone1Name=dyn_SkirtD01, bone2Name=pelvis, swing1LimitDegrees=5, swing2LimitDegrees=5, twistLimitDegrees=20}
 {bone1Name=dyn_SkirtD02, bone2Name=dyn_SkirtD01, swing1LimitDegrees=5, swing2LimitDegrees=5, twistLimitDegrees=20}
 {bone1Name=dyn_SkirtE01, bone2Name=pelvis, swing1LimitDegrees=5, swing2LimitDegrees=5, twistLimitDegrees=20}
 {bone1Name=dyn_SkirtE02, bone2Name=dyn_SkirtE01, swing1LimitDegrees=5, swing2LimitDegrees=5, twistLimitDegrees=20}
 {bone1Name=dyn_SkirtF01, bone2Name=pelvis, swing1LimitDegrees=5, swing2LimitDegrees=5, twistLimitDegrees=20}
 {bone1Name=dyn_SkirtF02, bone2Name=dyn_SkirtF01, swing1LimitDegrees=5, swing2LimitDegrees=5, twistLimitDegrees=20}
 {bone1Name=dyn_SkirtG01, bone2Name=pelvis, swing1LimitDegrees=5, swing2LimitDegrees=5, twistLimitDegrees=20}
 {bone1Name=dyn_SkirtG02, bone2Name=dyn_SkirtG01, swing1LimitDegrees=5, swing2LimitDegrees=5, twistLimitDegrees=20}
 {bone1Name=dyn_SkirtH01, bone2Name=pelvis, swing1LimitDegrees=5, swing2LimitDegrees=5, twistLimitDegrees=20}
 {bone1Name=dyn_SkirtH02, bone2Name=dyn_SkirtH01, swing1LimitDegrees=5, swing2LimitDegrees=5, twistLimitDegrees=20}
 {bone1Name=dyn_SkirtI01, bone2Name=pelvis, swing1LimitDegrees=5, swing2LimitDegrees=5, twistLimitDegrees=20}
 {bone1Name=dyn_SkirtI02, bone2Name=dyn_SkirtI01, swing1LimitDegrees=5, swing2LimitDegrees=5, twistLimitDegrees=20}
 {bone1Name=dyn_SkirtJ01, bone2Name=pelvis, swing1LimitDegrees=5, swing2LimitDegrees=5, twistLimitDegrees=20}
 {bone1Name=dyn_SkirtJ02, bone2Name=dyn_SkirtJ01, swing1LimitDegrees=5, swing2LimitDegrees=5, twistLimitDegrees=20}
 {bone1Name=dyn_Strap01, bone2Name=pelvis, swing1LimitDegrees=10, swing2LimitDegrees=10, twistLimitDegrees=30}
 {bone1Name=dyn_Strap02, bone2Name=dyn_Strap01, swing1LimitDegrees=10, swing2LimitDegrees=10, twistLimitDegrees=30}
 {bone1Name=dyn_Strap03, bone2Name=dyn_Strap02, swing1LimitDegrees=10, swing2LimitDegrees=10, twistLimitDegrees=30}
 {bone1Name=dyn_Strap04, bone2Name=dyn_Strap03, swing1LimitDegrees=10, swing2LimitDegrees=10, twistLimitDegrees=30}
 {bone1Name=dyn_Strap05, bone2Name=dyn_Strap04, swing1LimitDegrees=10, swing2LimitDegrees=10, twistLimitDegrees=30}
]
