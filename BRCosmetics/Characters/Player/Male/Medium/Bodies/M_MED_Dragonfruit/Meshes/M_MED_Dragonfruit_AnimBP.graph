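# Post-process style Anim Blueprint for the Dragonfruit male body cosmetic, correcting shoulder, forearm, neckline, skirt and strap bones over a pose copied from the attached parent mesh.
# Each update turns arm, thigh and head-mesh jaw deltas from reference pose into alphas.
# These drive chained constraints and additive bone offsets as pose-driven correctives.
# A pelvis-space RigidBody node then adds secondary physics, followed by final skirt and neckline fixes.
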AnimBlueprint /BRCosmetics/Characters/Player/Male/Medium/Bodies/M_MED_Dragonfruit/Meshes/M_MED_Dragonfruit_AnimBP
target_skeleton=&/M_MED_Dragonfruit_Skeleton
graphs [
 AnimGraph {nodes [
   "Transform (Modify) Bone - Bone: neck_NecklineShape"(Alpha=1, Translation="-1,2,0", Rotation="0,0,0", Scale="1,1,1.2") {
    rotationMode=BMM_Additive
    rotationSpace=BCS_BoneSpace
    scaleMode=BMM_Additive
    scaleSpace=BCS_BoneSpace
    translationMode=BMM_Additive
    translationSpace=BCS_BoneSpace
    alphaBoolBlend {blendOption=Linear}
    boneToModify {boneName=neck_NecklineShape}
    scale {z=1.2}
    translation=<-1 2>
   }
   n1 = "Constraint - flont_uppelalm_l (PP)"!(
    ComponentPose=Local To Component(LocalPose=Copy Pose From Mesh() {bUseAttachedParent=true})
    Alpha=1
    ConstraintWeights_0=0.5
    ConstraintWeights_1:float=0.5
   ) -> (Pose) {
    alphaBoolBlend {blendOption=Linear}
    boneToModify {boneName=flont_uppelalm_l}
    constraintSetup [{targetBone {boneName=upperarm_twist_01_l}}, {targetBone {boneName=clavicle_l}}]
    constraintWeights [1,1]
   }
   n2 = "Constraint - up_upperarm_r (PP)"!(ComponentPose=@n1, Alpha=1, ConstraintWeights_0=0.5, ConstraintWeights_1:float=0.5) -> (Pose) {
    alphaBoolBlend {blendOption=Linear}
    boneToModify {boneName=up_upperarm_r}
    constraintSetup [{targetBone {boneName=upperarm_twist_01_l}}, {targetBone {boneName=clavicle_l}}]
    constraintWeights [1,1]
   }
   n3 = "Constraint - front_upperarm_r (PP)"!(ComponentPose=@n2, Alpha=1, ConstraintWeights_0=0.5, ConstraintWeights_1:float=0.5) -> (Pose) {
    alphaBoolBlend {blendOption=Linear}
    boneToModify {boneName=front_upperarm_r}
    constraintSetup [{targetBone {boneName=upperarm_twist_01_r}}, {targetBone {boneName=clavicle_r}}]
    constraintWeights [1,1]
   }
   n4 = "Constraint - up_upperarm_r (PP)"(ComponentPose=@n3, Alpha=1, ConstraintWeights_0=0.5, ConstraintWeights_1=0.5) {
    alphaBoolBlend {blendOption=Linear}
    boneToModify {boneName=up_upperarm_r}
    constraintSetup [{targetBone {boneName=upperarm_twist_01_r}}, {targetBone {boneName=clavicle_r}}]
    constraintWeights [1,1]
   }
   n5 = "Transform (Modify) Bone - Bone: flont_uppelalm_l"(ComponentPose=@n4, Alpha=Get Shoulder_rot_f_l() -> (Shoulder_rot_f_l:double), Translation="-5,5,-5", Rotation="0,0,0", Scale="1,1,1") {
    rotationMode=BMM_Additive
    rotationSpace=BCS_BoneSpace
    scaleMode=BMM_Additive
    scaleSpace=BCS_BoneSpace
    translationMode=BMM_Additive
    translationSpace=BCS_BoneSpace
    alphaBoolBlend {blendOption=Linear}
    boneToModify {boneName=flont_uppelalm_l}
    translation=<-5 5 -5>
   }
   n6 = "Transform (Modify) Bone - Bone: front_upperarm_r"(ComponentPose=@n5, Alpha=Get Shoulder_rot_f_r() -> (Shoulder_rot_f_r:double), Translation="5,-5,-5", Rotation="0,0,0", Scale="1,1,1") {
    rotationMode=BMM_Additive
    rotationSpace=BCS_BoneSpace
    scaleMode=BMM_Additive
    scaleSpace=BCS_BoneSpace
    translationMode=BMM_Additive
    translationSpace=BCS_BoneSpace
    alphaBoolBlend {blendOption=Linear}
    boneToModify {boneName=front_upperarm_r}
    translation=<5 -5 -5>
   }
   n7 = "Transform (Modify) Bone - Bone: up_uppelalm_l"(ComponentPose=@n6, Alpha=Get Shoulder_rot_up_l() -> (Shoulder_rot_up_l:double), Translation="0,2,3", Rotation="0,0,0", Scale="1,1,1") {
    rotationMode=BMM_Additive
    rotationSpace=BCS_BoneSpace
    scaleMode=BMM_Additive
    scaleSpace=BCS_BoneSpace
    translationMode=BMM_Additive
    translationSpace=BCS_BoneSpace
    alphaBoolBlend {blendOption=Linear}
    boneToModify {boneName=up_uppelalm_l}
    translation {y=2, z=3}
   }
   n8 = "Transform (Modify) Bone - Bone: up_upperarm_r"(ComponentPose=@n7, Alpha=Get Shoulder_rot_up_r() -> (Shoulder_rot_up_r:double), Translation="0,-2,-3", Rotation="0,0,0", Scale="1,1,1") {
    rotationMode=BMM_Additive
    rotationSpace=BCS_BoneSpace
    scaleMode=BMM_Additive
    scaleSpace=BCS_BoneSpace
    translationMode=BMM_Additive
    translationSpace=BCS_BoneSpace
    alphaBoolBlend {blendOption=Linear}
    boneToModify {boneName=up_upperarm_r}
    translation {y=-2, z=-3}
   }
   n9 = "Transform (Modify) Bone - Bone: upperarmDn_r"(ComponentPose=@n8, Alpha=Get Shoulder_rot_dn_r() -> (Shoulder_rot_dn_r:double), Translation="4,0,0", Rotation="0,0,0", Scale="1,1,1") {
    rotationMode=BMM_Additive
    rotationSpace=BCS_BoneSpace
    scaleMode=BMM_Additive
    scaleSpace=BCS_BoneSpace
    translationMode=BMM_Additive
    translationSpace=BCS_BoneSpace
    alphaBoolBlend {blendOption=Linear}
    boneToModify {boneName=upperarmDn_r}
    translation {x=4}
   }
   n10 = "Transform (Modify) Bone - Bone: upperarmDn_l"(ComponentPose=@n9, Alpha=Get Shoulder_rot_dn_l() -> (Shoulder_rot_dn_l:double), Translation="-4,0,0", Rotation="0,0,0", Scale="1,1,1") {
    rotationMode=BMM_Additive
    rotationSpace=BCS_BoneSpace
    scaleMode=BMM_Additive
    scaleSpace=BCS_BoneSpace
    translationMode=BMM_Additive
    translationSpace=BCS_BoneSpace
    alphaBoolBlend {blendOption=Linear}
    boneToModify {boneName=upperarmDn_l}
    translation {x=-4}
   }
   n11 = "Transform (Modify) Bone - Bone: upperarmDn_l"(ComponentPose=@n10, Alpha=Get Shoulder_rot_up_l(), Translation="-5,0,8", Rotation="0,0,0", Scale="1,1,1") {
    rotationMode=BMM_Additive
    rotationSpace=BCS_BoneSpace
    scaleMode=BMM_Additive
    scaleSpace=BCS_BoneSpace
    translationMode=BMM_Additive
    translationSpace=BCS_BoneSpace
    alphaBoolBlend {blendOption=Linear}
    boneToModify {boneName=upperarmDn_l}
    translation {x=-5, z=8}
   }
   n12 = "Transform (Modify) Bone - Bone: upperarmDn_r"(ComponentPose=@n11, Alpha=Get Shoulder_rot_up_r(), Translation="5,0,-8", Rotation="0,0,0", Scale="1,1,1") {
    rotationMode=BMM_Additive
    rotationSpace=BCS_BoneSpace
    scaleMode=BMM_Additive
    scaleSpace=BCS_BoneSpace
    translationMode=BMM_Additive
    translationSpace=BCS_BoneSpace
    alphaBoolBlend {blendOption=Linear}
    boneToModify {boneName=upperarmDn_r}
    translation {x=5, z=-8}
   }
   n13 = "Constraint - upperarmDn_l (PP)"!(ComponentPose=@n12, Alpha=1, ConstraintWeights_0=0.4, ConstraintWeights_1:float=0.6) -> (Pose) {
    alphaBoolBlend {blendOption=Linear}
    boneToModify {boneName=upperarmDn_l}
    constraintSetup [{targetBone {boneName=spine_04}}, {targetBone {boneName=clavicle_l}}]
    constraintWeights [1,1]
   }
   n14 = "Constraint - upperarmDn_r (PP)"!(ComponentPose=@n13, Alpha=1, ConstraintWeights_0=0.4, ConstraintWeights_1:float=0.6) -> (Pose) {
    alphaBoolBlend {blendOption=Linear}
    boneToModify {boneName=upperarmDn_r}
    constraintSetup [{targetBone {boneName=spine_04}}, {targetBone {boneName=clavicle_r}}]
    constraintWeights [1,1]
   }
   n15 = "Constraint - lowerarm_l_front (RR)"!(ComponentPose=@n14, Alpha=1, ConstraintWeights_0=0.5, ConstraintWeights_1:float=0.5) -> (Pose) {
    alphaBoolBlend {blendOption=Linear}
    boneToModify {boneName=lowerarm_l_front}
    constraintSetup [{transformType=Rotation, targetBone {boneName=lowerarm_l}}, {transformType=Rotation, targetBone {boneName=upperarm_l}}]
    constraintWeights [1,1]
   }
   n16 = "Constraint - lowerarm_r_front (RR)"!(ComponentPose=@n15, Alpha=1, ConstraintWeights_0=0.5, ConstraintWeights_1:float=0.5) -> (Pose) {
    alphaBoolBlend {blendOption=Linear}
    boneToModify {boneName=lowerarm_r_front}
    constraintSetup [{transformType=Rotation, targetBone {boneName=lowerarm_r}}, {transformType=Rotation, targetBone {boneName=upperarm_r}}]
    constraintWeights [1,1]
   }
   n17 = "Constraint - lowerarm_l_back (RR)"!(ComponentPose=@n16, Alpha=1, ConstraintWeights_0=0.5, ConstraintWeights_1:float=0.5) -> (Pose) {
    alphaBoolBlend {blendOption=Linear}
    boneToModify {boneName=lowerarm_l_back}
    constraintSetup [{transformType=Rotation, targetBone {boneName=lowerarm_l}}, {transformType=Rotation, targetBone {boneName=upperarm_l}}]
    constraintWeights [1,1]
   }
   n18 = "Constraint - lowerarm_r_back (RR)"!(ComponentPose=@n17, Alpha=1, ConstraintWeights_0=0.5, ConstraintWeights_1:float=0.5) -> (Pose) {
    alphaBoolBlend {blendOption=Linear}
    boneToModify {boneName=lowerarm_r_back}
    constraintSetup [{transformType=Rotation, targetBone {boneName=lowerarm_r}}, {transformType=Rotation, targetBone {boneName=upperarm_r}}]
    constraintWeights [1,1]
   }
   n19 = Get lowerarm_l_front() -> (lowerarm_l_front:double)
   n20 = "Transform (Modify) Bone - Bone: lowerarm_l_front"(ComponentPose=@n18, Alpha=@n19, Translation="0,1,0", Rotation="0,0,0", Scale="1,1,1") {
    rotationMode=BMM_Additive
    rotationSpace=BCS_BoneSpace
    scaleMode=BMM_Additive
    scaleSpace=BCS_BoneSpace
    translationMode=BMM_Additive
    translationSpace=BCS_BoneSpace
    alphaBoolBlend {blendOption=Linear}
    boneToModify {boneName=lowerarm_l_front}
    translation {y=1}
   }
   n21 = Get lowerarm_r_front() -> (lowerarm_r_front:double)
   n22 = "Transform (Modify) Bone - Bone: lowerarm_r_front"(ComponentPose=@n20, Alpha=@n21, Translation="0,-1,0", Rotation="0,0,0", Scale="1,1,1") {
    rotationMode=BMM_Additive
    rotationSpace=BCS_BoneSpace
    scaleMode=BMM_Additive
    scaleSpace=BCS_BoneSpace
    translationMode=BMM_Additive
    translationSpace=BCS_BoneSpace
    alphaBoolBlend {blendOption=Linear}
    boneToModify {boneName=lowerarm_r_front}
    translation {y=-1}
   }
   n23 = "Transform (Modify) Bone - Bone: lowerarm_l_back"(ComponentPose=@n22, Alpha=@n19, Translation="0,-2,0", Rotation="0,0,0", Scale="1,1,1") {
    rotationMode=BMM_Additive
    rotationSpace=BCS_BoneSpace
    scaleMode=BMM_Additive
    scaleSpace=BCS_BoneSpace
    translationMode=BMM_Additive
    translationSpace=BCS_BoneSpace
    alphaBoolBlend {blendOption=Linear}
    boneToModify {boneName=lowerarm_l_back}
    translation {y=-2}
   }
   n24 = "Transform (Modify) Bone - Bone: lowerarm_r_back"(ComponentPose=@n23, Alpha=@n21, Translation="0,2,0", Rotation="0,0,0", Scale="1,1,1") {
    rotationMode=BMM_Additive
    rotationSpace=BCS_BoneSpace
    scaleMode=BMM_Additive
    scaleSpace=BCS_BoneSpace
    translationMode=BMM_Additive
    translationSpace=BCS_BoneSpace
    alphaBoolBlend {blendOption=Linear}
    boneToModify {boneName=lowerarm_r_back}
    translation {y=2}
   }
   n25 = Get leg_rot_l() -> (leg_rot_l:double)
   n26 = "Constraint - dyn_SkirtB01 (RP)"!(ComponentPose=@n24, Alpha=@n25, ConstraintWeights_0=1, ConstraintWeights_1:float=0) -> (Pose) {
    alphaBoolBlend {blendOption=Linear}
    boneToModify {boneName=dyn_SkirtB01}
    constraintSetup [{transformType=Rotation, targetBone {boneName=thigh_twist_01_l}}, {targetBone {boneName=pelvis}}]
    constraintWeights [1,1]
   }
   n27 = "Constraint - dyn_SkirtC01 (RR)"!(ComponentPose=@n26, Alpha=@n25, ConstraintWeights_0=0.8, ConstraintWeights_1:float=0.2) -> (Pose) {
    alphaBoolBlend {blendOption=Linear}
    boneToModify {boneName=dyn_SkirtC01}
    constraintSetup [{transformType=Rotation, targetBone {boneName=thigh_twist_01_l}}, {transformType=Rotation, targetBone {boneName=pelvis}}]
    constraintWeights [1,1]
   }
   n28 = "Transform (Modify) Bone - Bone: dyn_SkirtB01"(ComponentPose=@n27, Alpha=@n25, Translation="-5,8,5", Rotation="0,0,0", Scale="1,1,1") {
    rotationMode=BMM_Additive
    rotationSpace=BCS_BoneSpace
    scaleMode=BMM_Additive
    scaleSpace=BCS_BoneSpace
    translationMode=BMM_Additive
    translationSpace=BCS_BoneSpace
    alphaBoolBlend {blendOption=Linear}
    boneToModify {boneName=dyn_SkirtB01}
    translation=<-5 8 5>
   }
   n29 = "Transform (Modify) Bone - Bone: dyn_SkirtC01"(ComponentPose=@n28, Alpha=@n25, Translation="0,10,4", Rotation="0,0,0", Scale="1,1,1") {
    rotationMode=BMM_Additive
    rotationSpace=BCS_BoneSpace
    scaleMode=BMM_Additive
    scaleSpace=BCS_BoneSpace
    translationMode=BMM_Additive
    translationSpace=BCS_BoneSpace
    alphaBoolBlend {blendOption=Linear}
    boneToModify {boneName=dyn_SkirtC01}
    translation {y=10, z=4}
   }
   n30 = Get leg_rot_back_l() -> (leg_rot_back_l:double)
   n31 = "Constraint - dyn_SkirtD01 (RR)"!(ComponentPose=@n29, Alpha=@n30, ConstraintWeights_0=1, ConstraintWeights_1:float=0) -> (Pose) {
    alphaBoolBlend {blendOption=Linear}
    boneToModify {boneName=dyn_SkirtD01}
    constraintSetup [{transformType=Rotation, targetBone {boneName=thigh_twist_01_l}}, {transformType=Rotation, targetBone {boneName=pelvis}}]
    constraintWeights [1,1]
   }
   n32 = "Constraint - dyn_SkirtE01 (RR)"!(ComponentPose=@n31, Alpha=@n30, ConstraintWeights_0=1, ConstraintWeights_1:float=0) -> (Pose) {
    alphaBoolBlend {blendOption=Linear}
    boneToModify {boneName=dyn_SkirtE01}
    constraintSetup [{transformType=Rotation, targetBone {boneName=thigh_twist_01_l}}, {transformType=Rotation, targetBone {boneName=pelvis}}]
    constraintWeights [1,1]
   }
   n33 = Get leg_rot_r() -> (leg_rot_r:double)
   n34 = "Constraint - dyn_SkirtJ01 (RP)"!(ComponentPose=@n32, Alpha=@n33, ConstraintWeights_0=1, ConstraintWeights_1:float=0) -> (Pose) {
    alphaBoolBlend {blendOption=Linear}
    boneToModify {boneName=dyn_SkirtJ01}
    constraintSetup [{transformType=Rotation, targetBone {boneName=thigh_twist_01_r}}, {targetBone {boneName=pelvis}}]
    constraintWeights [1,1]
   }
   n35 = "Constraint - dyn_SkirtI01 (RR)"!(ComponentPose=@n34, Alpha=@n33, ConstraintWeights_0=0.8, ConstraintWeights_1:float=0.2) -> (Pose) {
    alphaBoolBlend {blendOption=Linear}
    boneToModify {boneName=dyn_SkirtI01}
    constraintSetup [{transformType=Rotation, targetBone {boneName=thigh_twist_01_r}}, {transformType=Rotation, targetBone {boneName=pelvis}}]
    constraintWeights [1,1]
   }
   n36 = "Transform (Modify) Bone - Bone: dyn_SkirtJ01"(ComponentPose=@n35, Alpha=@n33, Translation="5,8,5", Rotation="0,0,10", Scale="1,1,1") {
    rotationMode=BMM_Additive
    rotationSpace=BCS_BoneSpace
    scaleMode=BMM_Additive
    scaleSpace=BCS_BoneSpace
    translationMode=BMM_Additive
    translationSpace=BCS_BoneSpace
    alphaBoolBlend {blendOption=Linear}
    boneToModify {boneName=dyn_SkirtJ01}
    rotation {roll=10}
    translation=<5 8 5>
   }
   n37 = "Transform (Modify) Bone - Bone: dyn_SkirtI01"(ComponentPose=@n36, Alpha=@n33, Translation="0,10,4", Rotation="0,0,0", Scale="1,1,1") {
    rotationMode=BMM_Additive
    rotationSpace=BCS_BoneSpace
    scaleMode=BMM_Additive
    scaleSpace=BCS_BoneSpace
    translationMode=BMM_Additive
    translationSpace=BCS_BoneSpace
    alphaBoolBlend {blendOption=Linear}
    boneToModify {boneName=dyn_SkirtI01}
    translation {y=10, z=4}
   }
   n38 = Get leg_rot_back_r() -> (leg_rot_back_r:double)
   n39 = "Constraint - dyn_SkirtH01 (RR)"!(ComponentPose=@n37, Alpha=@n38, ConstraintWeights_0=1, ConstraintWeights_1:float=0) -> (Pose) {
    alphaBoolBlend {blendOption=Linear}
    boneToModify {boneName=dyn_SkirtH01}
    constraintSetup [{transformType=Rotation, targetBone {boneName=thigh_twist_01_r}}, {transformType=Rotation, targetBone {boneName=pelvis}}]
    constraintWeights [1,1]
   }
   n40 = "Constraint - dyn_SkirtG01 (RR)"!(ComponentPose=@n39, Alpha=@n38, ConstraintWeights_0=1, ConstraintWeights_1:float=0) -> (Pose) {
    alphaBoolBlend {blendOption=Linear}
    boneToModify {boneName=dyn_SkirtG01}
    constraintSetup [{transformType=Rotation, targetBone {boneName=thigh_twist_01_r}}, {transformType=Rotation, targetBone {boneName=pelvis}}]
    constraintWeights [1,1]
   }
   n41 = "Constraint - dyn_SkirtA01 (PP)"!(ComponentPose=@n40, Alpha=1, ConstraintWeights_0=0.5, ConstraintWeights_1:float=0.5) -> (Pose) {
    alphaBoolBlend {blendOption=Linear}
    boneToModify {boneName=dyn_SkirtA01}
    constraintSetup [{targetBone {boneName=dyn_SkirtB01}}, {targetBone {boneName=dyn_SkirtJ01}}]
    constraintWeights [1,1]
   }
   n42 = "Transform (Modify) Bone - Bone: dyn_SkirtA01"(ComponentPose=@n41, Alpha=Get leg_rot_l(), Translation="0,4.5,0", Rotation="0,0,0", Scale="1,1,1") {
    rotationMode=BMM_Additive
    rotationSpace=BCS_BoneSpace
    scaleMode=BMM_Additive
    scaleSpace=BCS_BoneSpace
    translationMode=BMM_Additive
    translationSpace=BCS_BoneSpace
    alphaBoolBlend {blendOption=Linear}
    boneToModify {boneName=dyn_SkirtA01}
    translation {y=4.5}
   }
   n43 = "Constraint - dyn_SkirtF01 (PP)"!(ComponentPose=@n42, Alpha=1, ConstraintWeights_0=0.5, ConstraintWeights_1:float=0.5) -> (Pose) {
    alphaBoolBlend {blendOption=Linear}
    boneToModify {boneName=dyn_SkirtF01}
    constraintSetup [{targetBone {boneName=dyn_SkirtG01}}, {targetBone {boneName=dyn_SkirtE01}}]
    constraintWeights [1,1]
   }
   n44 = Get leg_rot_r()
   n45 = "Constraint - dyn_Strap01 (PP)"!(ComponentPose=@n43, Alpha=@n44, ConstraintWeights_0=0.6, ConstraintWeights_1:float=0.4) -> (Pose) {
    alphaBoolBlend {blendOption=Linear}
    boneToModify {boneName=dyn_Strap01}
    constraintSetup [{targetBone {boneName=dyn_SkirtJ01}}, {targetBone {boneName=dyn_Strap01}}]
    constraintWeights [1,1]
   }
   n46 = "Constraint - dyn_Strap02 (PPP)"!(ComponentPose=@n45, Alpha=1, ConstraintWeights_0=0.2, ConstraintWeights_1:float=0.2, ConstraintWeights_2:float=0.6) -> (Pose) {
    alphaBoolBlend {blendOption=Linear}
    boneToModify {boneName=dyn_Strap02}
    constraintSetup [{targetBone {boneName=dyn_SkirtJ02}}, {targetBone {boneName=dyn_SkirtI02}}, {targetBone {boneName=dyn_Strap02}}]
    constraintWeights [1,1,1]
   }
   n47 = "Transform (Modify) Bone - Bone: dyn_Strap01"(ComponentPose=@n46, Alpha=@n44, Translation="0,3,0", Rotation="0,-30,-20", Scale="1,1,1") {
    rotationMode=BMM_Additive
    rotationSpace=BCS_BoneSpace
    scaleMode=BMM_Additive
    scaleSpace=BCS_BoneSpace
    translationMode=BMM_Additive
    translationSpace=BCS_BoneSpace
    alphaBoolBlend {blendOption=Linear}
    boneToModify {boneName=dyn_Strap01}
    rotation {roll=-20, yaw=-30}
    translation {y=3}
   }
   n48 = "Transform (Modify) Bone - Bone: dyn_Strap02"(ComponentPose=@n47, Alpha=@n44, Translation="0,3,-1", Rotation="0,0,0", Scale="1,1,1") {
    rotationMode=BMM_Additive
    rotationSpace=BCS_BoneSpace
    scaleMode=BMM_Additive
    scaleSpace=BCS_BoneSpace
    translationMode=BMM_Additive
    translationSpace=BCS_BoneSpace
    alphaBoolBlend {blendOption=Linear}
    boneToModify {boneName=dyn_Strap02}
    translation {y=3, z=-1}
   }
   n49 = RigidBody(ComponentPose=@n48, Alpha=0.9) {
    bClampLinearTranslationLimitToRefPose=true
    bOverrideWorldGravity=true
    overridePhysicsAsset=&/M_MED_Dragonfruit_Physics
    simulationSpace=BaseBoneSpace
    alphaBoolBlend {blendOption=Linear}
    baseBoneRef {boneName=pelvis}
    componentAppliedLinearAccClamp=<2000 2000 2000>
    componentLinearAccScale=<0.35 0.35 0.35>
    componentLinearVelScale=<0.3 0.3 0.3>
    overrideWorldGravity {z=-980}
   }
   n50 = "Constraint - dyn_SkirtA02 (PP)"!(ComponentPose=@n49, Alpha=1, ConstraintWeights_0=0.5, ConstraintWeights_1:float=0.5) -> (Pose) {
    alphaBoolBlend {blendOption=Linear}
    boneToModify {boneName=dyn_SkirtA02}
    constraintSetup [{targetBone {boneName=dyn_SkirtB02}}, {targetBone {boneName=dyn_SkirtJ02}}]
    constraintWeights [1,1]
   }
   n51 = "Transform (Modify) Bone - Bone: dyn_SkirtC02"(ComponentPose=@n50, Alpha=Get leg_rot_l(), Translation="0,-3,0", Rotation="0,0,-35", Scale="1,1,1") {
    rotationMode=BMM_Additive
    rotationSpace=BCS_BoneSpace
    scaleMode=BMM_Additive
    scaleSpace=BCS_BoneSpace
    translationMode=BMM_Additive
    translationSpace=BCS_BoneSpace
    alphaBoolBlend {blendOption=Linear}
    boneToModify {boneName=dyn_SkirtC02}
    rotation {roll=-35}
    translation {y=-3}
   }
   n52 = "Transform (Modify) Bone - Bone: dyn_SkirtI02"(ComponentPose=@n51, Alpha=Get leg_rot_r(), Translation="0,-3,0", Rotation="0,0,-35", Scale="1,1,1") {
    rotationMode=BMM_Additive
    rotationSpace=BCS_BoneSpace
    scaleMode=BMM_Additive
    scaleSpace=BCS_BoneSpace
    translationMode=BMM_Additive
    translationSpace=BCS_BoneSpace
    alphaBoolBlend {blendOption=Linear}
    boneToModify {boneName=dyn_SkirtI02}
    rotation {roll=-35}
    translation {y=-3}
   }
   n53 = Get neck_rot() -> (neck_rot:double)
   n54 = "Constraint - neck_NecklineShape (PP)"!(ComponentPose=@n52, Alpha=@n53, ConstraintWeights_0=0.7, ConstraintWeights_1:float=0.3) -> (Pose) {
    alphaBoolBlend {blendOption=Linear}
    boneToModify {boneName=neck_NecklineShape}
    constraintSetup [{targetBone {boneName=neck_02}}, {targetBone {boneName=spine_05}}]
    constraintWeights [1,1]
   }
   n55 = "Transform (Modify) Bone - Bone: neck_NecklineShape"(ComponentPose=@n54, Alpha=@n53, Translation="0,1,0", Rotation="0,0,0", Scale="1,1,1") {
    rotationMode=BMM_Additive
    rotationSpace=BCS_BoneSpace
    scaleMode=BMM_Additive
    scaleSpace=BCS_BoneSpace
    translationMode=BMM_Additive
    translationSpace=BCS_BoneSpace
    alphaBoolBlend {blendOption=Linear}
    boneToModify {boneName=neck_NecklineShape}
    translation {y=1}
   }
   n56 = "Transform (Modify) Bone - Bone: dyn_SkirtI01"(ComponentPose=@n55, Alpha=1, Translation="0,0,0", Rotation="0.000,0.000,-10", Scale="1,1,1") {
    rotationMode=BMM_Additive
    rotationSpace=BCS_BoneSpace
    scaleMode=BMM_Additive
    scaleSpace=BCS_BoneSpace
    translationMode=BMM_Additive
    translationSpace=BCS_BoneSpace
    alphaBoolBlend {blendOption=Linear}
    boneToModify {boneName=dyn_SkirtI01}
    rotation {roll=-10}
   }
   Output Pose(Result=Component To Local(ComponentPose=@n56)) {layerGroup=None}
  ]}
 EventGraph {nodes [
   Event BlueprintUpdateAnimation()
   n1 = Get Skel_Mesh()
   > Set Shoulder_rot_f_r(Shoulder_rot_f_r:double=GetAlphaFromReferenceTransformDelta(SkeletalMeshComponent=@n1, BoneName=upperarm_r, BaseName=clavicle_r, InputRange=(X=0,Y=90), TwistAxis="0,0,1")) -> (Output_Get:double)
   > Set Shoulder_rot_f_l(Shoulder_rot_f_l:double=GetAlphaFromReferenceTransformDelta(SkeletalMeshComponent=@n1, BoneName=upperarm_l, BaseName=clavicle_l, InputRange=(X=0,Y=90), TwistAxis="0,0,1")) -> (Output_Get:double)
   n2 = Get Skel_Mesh()
   > Set Shoulder_rot_up_r(Shoulder_rot_up_r:double=GetAlphaFromReferenceTransformDelta(
     SkeletalMeshComponent=@n2
     BoneName=upperarm_twist_01_r
     BaseName=clavicle_r
     DeltaComponentType=EulerY
     InputRange=(X=0,Y=110)
     TwistAxis="0,0,1"
    )) -> (Output_Get:double)
   > Set Shoulder_rot_up_l(Shoulder_rot_up_l:double=GetAlphaFromReferenceTransformDelta(SkeletalMeshComponent=@n2, BoneName=upperarm_l, BaseName=clavicle_l, DeltaComponentType=EulerY, InputRange=(X=0,Y=110), TwistAxis="0,0,1")) -> (Output_Get:double)
   n3 = Get Skel_Mesh()
   > Set Shoulder_rot_dn_r(Shoulder_rot_dn_r:double=GetAlphaFromReferenceTransformDelta(SkeletalMeshComponent=@n3, BoneName=upperarm_r, BaseName=clavicle_r, InputRange=(X=-15,Y=-90), TwistAxis="0,0,1")) -> (Output_Get:double)
   > Set Shoulder_rot_dn_l(Shoulder_rot_dn_l:double=GetAlphaFromReferenceTransformDelta(SkeletalMeshComponent=@n3, BoneName=upperarm_l, BaseName=clavicle_l, InputRange=(X=-15,Y=-90), TwistAxis="0,0,1")) -> (Output_Get:double)
   > n4 = Sequence!(execute) -> (then_0, then_1, then_2:Exec, then_3:Exec)
   n5 = Get Skel_Mesh()
   Set leg_rot_r(
    execute=@n4.then_2
    leg_rot_r:double=GetAlphaFromReferenceTransformDelta(SkeletalMeshComponent=@n5, BoneName=thigh_twist_01_r, BaseName=pelvis, InputRange=(X=0,Y=-60), TwistAxis="0,0,1")
   ) -> (Output_Get:double)
   > Set leg_rot_l(leg_rot_l:double=GetAlphaFromReferenceTransformDelta(SkeletalMeshComponent=@n5, BoneName=thigh_twist_01_l, BaseName=pelvis, InputRange=(X=0,Y=-60), TwistAxis="0,0,1")) -> (Output_Get:double)
   n6 = Get Skel_Mesh()
   > Set leg_rot_back_r(leg_rot_back_r:double=GetAlphaFromReferenceTransformDelta(SkeletalMeshComponent=@n6, BoneName=thigh_r, BaseName=pelvis, InputRange=(X=0,Y=50), TwistAxis="0,0,1")) -> (Output_Get:double)
   > Set leg_rot_back_l(leg_rot_back_l:double=GetAlphaFromReferenceTransformDelta(SkeletalMeshComponent=@n6, BoneName=thigh_l, BaseName=pelvis, InputRange=(X=0,Y=50), TwistAxis="0,0,1")) -> (Output_Get:double)
   > Set neck_rot(neck_rot:double=GetAlphaFromReferenceTransformDelta(
     SkeletalMeshComponent=Get Head_Mesh() -> (Head_Mesh:Skeletal Mesh Component Object Reference)
     BoneName=C_jaw
     BaseName=spine_05
     DeltaComponentType=EulerX
     InputRange=(X=30,Y=-30)
     TwistAxis="0,0,1"
    )) -> (Output_Get:double)
   Set lowerarm_l_front(
    execute=@n4.then_0
    lowerarm_l_front:double=GetAlphaFromReferenceTransformDelta(SkeletalMeshComponent=Get Skel_Mesh(), BoneName=upperarm_l, BaseName=lowerarm_l, InputRange=(X=0,Y=-90), TwistAxis="0,0,1")
   ) -> (Output_Get:double)
   Set lowerarm_r_front(
    execute=@n4.then_1
    lowerarm_r_front:double=GetAlphaFromReferenceTransformDelta(SkeletalMeshComponent=Get Skel_Mesh(), BoneName=upperarm_r, BaseName=lowerarm_r, InputRange=(X=0,Y=-90), TwistAxis="0,0,1")
   ) -> (Output_Get:double)
   Event BlueprintInitializeAnimation()
   > n7 = Sequence!(execute) -> (then_0, then_1)
   n8 = Get bRigidBodyEnabled()
   > Set frontend_rban_disabled(frontend_rban_disabled:bool=AND Boolean(A=Get bIsInFrontEnd(), B=NOT Boolean(A=@n8))) -> (Output_Get:bool)
   Set Skel_Mesh(execute=@n7.then_1, Skel_Mesh=GetOwnerSkeletalMeshComponent())
   > n9 = Cast To PlayerPawn_Athena(Object=TryGetPawnOwner())
   > n10 = Set Pawn Owner(Pawn Owner:Fort Player Pawn Athena Object Reference=@n9) -> (Output_Get:Fort Player Pawn Athena Object Reference)
   > Set Head_Mesh(Head_Mesh:Skeletal Mesh Component Object Reference=GetSkeletalMeshForPartType(self=@n10, PartType=Head)) -> (Output_Get:Skeletal Mesh Component Object Reference)
   Get Skel_Mesh()
  ]}
]
variables [
 frontend_rban_disabled {type=boolean}
 Head_Mesh {type=SkeletalMeshComponent}
 leg_rot_back_l {type=number}
 leg_rot_back_r {type=number}
 leg_rot_l {type=number}
 leg_rot_r {type=number}
 lowerarm_l_back {type=number}
 lowerarm_l_front {type=number}
 lowerarm_r_back {type=number}
 lowerarm_r_front {type=number}
 neck_rot {type=number}
 Pawn Owner {type=FortPlayerPawnAthena}
 Shoulder_rot_dn_l {type=number}
 Shoulder_rot_dn_r {type=number}
 Shoulder_rot_f_l {type=number}
 Shoulder_rot_f_r {type=number}
 Shoulder_rot_up_l {type=number}
 Shoulder_rot_up_r {type=number}
 Skel_Mesh {type=SkeletalMeshComponent}
]
