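# Physics asset for the face accessory mesh of the Dragonfruit medium male player cosmetic, defining physics bodies for its dynamic hair bones.
# Single-capsule bodies cover the head, a hair base and two hair pieces, A and B.
# The base is constrained almost rigidly to the head, while A and B branch separately from the base with wider angular limits, B swinging more freely than A.
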PhysicsAsset /BRCosmetics/Characters/Player/Male/Medium/Bodies/M_MED_Dragonfruit/Meshes/Parts/M_MED_Dragonfruit_FaceAcc_Physics
bodies {
 dyn_hair_A [{
   shapeType=Capsule
   length=3.6071
   radius=2.4347
   shapeName=dyn_hair_A_capsule
   center=<-0.1223 -0.3386 3.414>
   rotation=<rot 0 -5.5455 -3.8441>
  }]
 dyn_hair_B [{
   shapeType=Capsule
   length=5.1707
   radius=2.1538
   shapeName=dyn_hair_B_capsule
   center=<0.3214 -2.3865 2.8457>
   rotation=<rot -0.0001 -178.8065 34.0881>
  }]
 dyn_hair_base [{shapeType=Capsule, length=3.5, radius=6.75, shapeName=dyn_hair_base_capsule, rotation=<rot -89.831 89.9333 179.9994>}]
 head [{shapeType=Capsule, length=6.0702, radius=6.6802, center=<2.835 2.3982 0.1093>, rotation=<rot 87.1998 -52.28 129.2263>}]
}
constraints [
 {bone1Name=dyn_hair_base, bone2Name=head, swing1LimitDegrees=1, swing2LimitDegrees=1, twistLimitDegrees=1}
 {bone1Name=dyn_hair_A, bone2Name=dyn_hair_base, swing1LimitDegrees=10, swing2LimitDegrees=10, twistLimitDegrees=10}
 {bone1Name=dyn_hair_B, bone2Name=dyn_hair_base, swing1LimitDegrees=20, swing2LimitDegrees=20, twistLimitDegrees=15}
]
